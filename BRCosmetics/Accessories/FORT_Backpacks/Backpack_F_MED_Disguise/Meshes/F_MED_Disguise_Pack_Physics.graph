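# Physics asset for the F_MED_Disguise backpack cosmetic, defining secondary motion for its candies, lollipops, ribbons and strings.
# Each dynamic bone gets one capsule, and most are constrained directly to the small dyn_candyBase hub.
# Two string chains (five links left, four right) and a two-link big-candy chain hang from that hub.
# Angular limits keep small pieces nearly rigid while ribbons, long candy and strings swing more freely.
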PhysicsAsset /BRCosmetics/Accessories/FORT_Backpacks/Backpack_F_MED_Disguise/Meshes/F_MED_Disguise_Pack_Physics
bodies {
 dyn_CandyBig_03 [{
   shapeType=Capsule
   length=5.3751
   radius=1.3656
   shapeName=dyn_CandyBig_03_capsule
   center=<-2.4518 -1.5071 -2.1257>
   rotation=<rot 1.1259 -49.3627 -90>
  }]
 dyn_candyBase [{
   shapeType=Capsule
   length=0.5
   radius=0.5936
   shapeName=dyn_candyBase_capsule
   center=<-0.1519 0.0332 0.1144>
   rotation=<rot -9.7096 77.7673 -89.9995>
  }]
 dyn_candyBig_01 [{
   shapeType=Capsule
   length=6.8897
   radius=1.5832
   shapeName=dyn_candyBig_01_capsule
   center=<-2.9417 0.6525 0.4094>
   rotation=<rot 0 8.5372 -84.454>
  }]
 dyn_candyBig_02 [{
   shapeType=Capsule
   length=6.3071
   radius=1.7229
   shapeName=dyn_candyBig_02_capsule
   center=<-5.1082 -0.3171 -1.2384>
   rotation=<rot 0 14.2774 -78.7087>
  }]
 dyn_candyBig_04 [{
   shapeType=Capsule
   length=3.2647
   radius=1.3578
   shapeName=dyn_candyBig_04_capsule
   center=<1.3857 0.2022 0.1721>
   rotation=<rot 0 -171.9198 -81.7014>
  }]
 dyn_candyBig_05 [{
   shapeType=Capsule
   length=4.4492
   radius=2.7915
   shapeName=dyn_candyBig_05_capsule
   center=<2.1469 0.1945 -1.051>
   rotation=<rot 0 1.4476 -79.0243>
  }]
 dyn_candyCane [{
   shapeType=Capsule
   length=5.5684
   radius=2.385
   shapeName=dyn_candyCane_capsule
   center=<2.6787 -0.4706 -0.0302>
   rotation=<rot 0 45.1977 -83.96>
  }]
 dyn_candyLong [{
   shapeType=Capsule
   length=2.4714
   radius=0.9093
   shapeName=dyn_candyLong_capsule
   center=<1.7472 1.1535 -0.8088>
   rotation=<rot 0 -66.14 -73.2117>
  }]
 dyn_candyPieces_01 [{
   shapeType=Capsule
   length=1.5976
   radius=1.1117
   shapeName=dyn_candyPieces_01_capsule
   center=<0.3129 0.5899 -1.2181>
   rotation=<rot 0 -5.0061 -85.0274>
  }]
 dyn_candyPieces_02 [{
   shapeType=Capsule
   length=1.3705
   radius=0.9787
   shapeName=dyn_candyPieces_02_capsule
   center=<-2.1395 -1.9921 -1.1375>
   rotation=<rot 0 142.1731 -51.7276>
  }]
 dyn_candyPieces_03 [{
   shapeType=Capsule
   length=1.2107
   radius=0.8604
   shapeName=dyn_candyPieces_03_capsule
   center=<-0.2273 -0.1899 -0.1492>
   rotation=<rot 0 115.9699 -62.4144>
  }]
 dyn_candyPieces_04 [{
   shapeType=Capsule
   length=2.1403
   radius=0.8577
   shapeName=dyn_candyPieces_04_capsule
   center=<0.4655 0.7286 1.1764>
   rotation=<rot -40.6859 -90 21.8142>
  }]
 dyn_candySmall_01 [{
   shapeType=Capsule
   length=3.2916
   radius=0.8732
   shapeName=dyn_candySmall_01_capsule
   center=<0.0806 -0.0864 0.0753>
   rotation=<rot 0 8.2695 -82.3512>
  }]
 dyn_candySmall_02 [{
   shapeType=Capsule
   length=3.1296
   radius=1.054
   shapeName=dyn_candySmall_02_capsule
   center=<0.0569 0.0138 -0.411>
   rotation=<rot 0 -156.2204 -89.7912>
  }]
 dyn_lollipopBig [{
   shapeType=Capsule
   length=3.9538
   radius=3.7755
   shapeName=dyn_lollipopBig_capsule
   center=<-0.2787 -0.0662 0.3267>
   rotation=<rot 0 106.4906 -62.5771>
  }]
 dyn_lollipopSmall [{
   shapeType=Capsule
   length=2.6746
   radius=2.6445
   shapeName=dyn_lollipopSmall_capsule
   center=<0.0628 0.053 0.2897>
   rotation=<rot 0 -25.3321 -79.7761>
  }]
 dyn_ribbonThick [{
   shapeType=Capsule
   length=3.0045
   radius=1.0026
   shapeName=dyn_ribbonThick_capsule
   center=<-1.575 1.8282 0.2426>
   rotation=<rot 0 -134.2309 -61.0107>
  }]
 dyn_ribbonThin [{
   shapeType=Capsule
   length=1.8476
   radius=0.806
   shapeName=dyn_ribbonThin_capsule
   center=<0.6754 1.3111 0.5935>
   rotation=<rot 0 142.3653 -57.0936>
  }]
 dyn_string_01_LFT [{shapeType=Capsule, length=0.51, radius=0.5, shapeName=dyn_string_01_LFT_capsule, rotation=<rot 90 0 0>}]
 dyn_string_01_RGT [{shapeType=Capsule, length=0.6269, radius=0.3931, shapeName=dyn_string_01_RGT_capsule, rotation=<rot 90 0 0>}]
 dyn_string_02_LFT [{shapeType=Capsule, length=2.005, radius=0.505, shapeName=dyn_string_02_LFT_capsule, center=<1.5002 0 0>, rotation=<rot 90 0 0>}]
 dyn_string_02_RGT [{shapeType=Capsule, length=2.5031, radius=0.4287, shapeName=dyn_string_02_RGT_capsule, center=<1.5957 0 0>, rotation=<rot 90 0 0>}]
 dyn_string_03_LFT [{shapeType=Capsule, length=2.005, radius=1.005, shapeName=dyn_string_03_LFT_capsule, center=<1.5577 0 0>, rotation=<rot 90 0 0>}]
 dyn_string_03_RGT [{
   shapeType=Capsule
   length=2.88
   radius=0.63
   shapeName=dyn_string_03_RGT_capsule
   center=<1.4149 -0.0001 0.1853>
   rotation=<rot 78.2781 -179.9999 -179.9999>
  }]
 dyn_string_04_LFT [{shapeType=Capsule, length=1.005, radius=1.005, shapeName=dyn_string_04_LFT_capsule, center=<1.4275 0 0>, rotation=<rot 90 0 0>}]
 dyn_string_04_RGT [{shapeType=Capsule, length=1.2336, radius=1, shapeName=dyn_string_04_RGT_capsule, center=<1.4651 0 0>, rotation=<rot 90 0 0>}]
 dyn_string_05_LFT [{shapeType=Capsule, length=0.9416, radius=1.1403, shapeName=dyn_string_05_LFT_capsule, center=<1.0911 0 0>, rotation=<rot 90 0 0>}]
}
constraints [
 {bone1Name=dyn_candyBig_01, bone2Name=dyn_candyBase, swing1LimitDegrees=15, swing2LimitDegrees=15, twistLimitDegrees=5}
 {bone1Name=dyn_candyBig_02, bone2Name=dyn_candyBase, swing1LimitDegrees=10, swing2LimitDegrees=10, twistLimitDegrees=5}
 {bone1Name=dyn_CandyBig_03, bone2Name=dyn_candyBase, swing1LimitDegrees=10, swing2LimitDegrees=10, twistLimitDegrees=2.5}
 {bone1Name=dyn_candySmall_01, bone2Name=dyn_candyBase, swing1LimitDegrees=10, swing2LimitDegrees=10, twistLimitDegrees=10}
 {bone1Name=dyn_candySmall_02, bone2Name=dyn_candyBase, swing1LimitDegrees=10, swing2LimitDegrees=10, twistLimitDegrees=10}
 {bone1Name=dyn_candyPieces_01, bone2Name=dyn_candyBase, swing1LimitDegrees=2, swing2LimitDegrees=2, twistLimitDegrees=1}
 {bone1Name=dyn_candyPieces_02, bone2Name=dyn_candyBase, swing1LimitDegrees=2, swing2LimitDegrees=2, twistLimitDegrees=1}
 {bone1Name=dyn_candyPieces_03, bone2Name=dyn_candyBase, swing1LimitDegrees=2, swing2LimitDegrees=2, twistLimitDegrees=1}
 {bone1Name=dyn_candyPieces_04, bone2Name=dyn_candyBase, swing1LimitDegrees=2, swing2LimitDegrees=2, twistLimitDegrees=1}
 {bone1Name=dyn_ribbonThin, bone2Name=dyn_candyBase, swing1LimitDegrees=15, swing2LimitDegrees=35, twistLimitDegrees=5}
 {bone1Name=dyn_ribbonThick, bone2Name=dyn_candyBase, swing1LimitDegrees=15, swing2LimitDegrees=35, twistLimitDegrees=5}
 {bone1Name=dyn_candyLong, bone2Name=dyn_candyBase, swing1LimitDegrees=15, swing2LimitDegrees=35, twistLimitDegrees=5}
 {bone1Name=dyn_candyCane, bone2Name=dyn_candyBase, swing1LimitDegrees=7.5, swing2LimitDegrees=5, twistLimitDegrees=5}
 {bone1Name=dyn_lollipopBig, bone2Name=dyn_candyBase, swing1LimitDegrees=5, swing2LimitDegrees=5, twistLimitDegrees=1}
 {bone1Name=dyn_lollipopSmall, bone2Name=dyn_candyBase, swing1LimitDegrees=5, swing2LimitDegrees=5, twistLimitDegrees=5}
 {bone1Name=dyn_string_01_RGT, bone2Name=dyn_candyBase, swing1LimitDegrees=30, swing2LimitDegrees=5, twistLimitDegrees=5}
 {bone1Name=dyn_string_02_RGT, bone2Name=dyn_string_01_RGT, swing1LimitDegrees=30, swing2LimitDegrees=20, twistLimitDegrees=5}
 {bone1Name=dyn_string_03_RGT, bone2Name=dyn_string_02_RGT, swing1LimitDegrees=30, swing2LimitDegrees=20, twistLimitDegrees=5}
 {bone1Name=dyn_string_04_RGT, bone2Name=dyn_string_03_RGT, swing1LimitDegrees=30, swing2LimitDegrees=20, twistLimitDegrees=5}
 {bone1Name=dyn_string_01_LFT, bone2Name=dyn_candyBase, swing1LimitDegrees=30, swing2LimitDegrees=5, twistLimitDegrees=5}
 {bone1Name=dyn_string_02_LFT, bone2Name=dyn_string_01_LFT, swing1LimitDegrees=30, swing2LimitDegrees=20, twistLimitDegrees=5}
 {bone1Name=dyn_string_03_LFT, bone2Name=dyn_string_02_LFT, swing1LimitDegrees=30, swing2LimitDegrees=25, twistLimitDegrees=5}
 {bone1Name=dyn_string_04_LFT, bone2Name=dyn_string_03_LFT, swing1LimitDegrees=30, swing2LimitDegrees=25, twistLimitDegrees=5}
 {bone1Name=dyn_string_05_LFT, bone2Name=dyn_string_04_LFT, swing1LimitDegrees=30, swing2LimitDegrees=25, twistLimitDegrees=5}
 {bone1Name=dyn_candyBig_04, bone2Name=dyn_candyBase, swing1LimitDegrees=15, swing2LimitDegrees=15, twistLimitDegrees=5}
 {bone1Name=dyn_candyBig_05, bone2Name=dyn_candyBig_04, swing1LimitDegrees=15, swing2LimitDegrees=15, twistLimitDegrees=5}
]
